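; Season 16 quest gameplay ability that awards progress when the player holds shield at a threshold for a set time.
; After a startup delay, the server watches shield changes, ending the ability in Phosphorus playlists.
; A briefly delayed recheck, guarding against momentary LTM shield flashes, starts a timer and grants a transient countdown quest.
; Expiry sends the quest stat event and ends the ability; dropping below threshold clears the timer and resumes monitoring.
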(event Ability|EventFailedToActivatePassiveAbility
  (params (ActorInfo :type Gameplay%20Ability%20Actor%20Info%20Structure))
  (|CallParentFunction))

(event Custom|ContinueMonitoring
  ;[1 1 1] Wait for attribute changes and check progress again
  (bind (_asynctaskproxy _bmatchescomparison) (Ability|Tasks|WaitForAttributeChangeThreshold (Variables|Default|GetAtt_Shield) "GreaterThanOrEqualTo" (Variables|Default|GetComparisonValue))
    (:OnChange
      (Variables|Default|SetShieldSuccess _bmatchescomparison)
      (if (GameplayTags|MatchesAnyTags "(TagName=\"Athena.Playlist.Phosphorus\")" (GameContext|GetGameContextGameplayTags))
        (Ability|EndAbility)
        (elif (Variables|Default|GetCountdownStarted)
          (if _bmatchescomparison
            (CallFunction|TryProgressQuest)
            (else
              (CallFunction|EndTimer)))
          (elif _bmatchescomparison
            (CallFunction|ResetCheck)
            (CallFunction|TryProgressQuest)
            (else
              (CallFunction|EndTimer))))))))

(event Ability|EventActivateAbility
  ;[1 1 1] Start ability
  (|CallParentFunction)
  (Utilities|FlowControl|Delay 1e+01)
  (if (Networking|IsServer)
    ;[1 1 1] Wait for attribute changes and check progress again
    (bind (_asynctaskproxy _bmatchescomparison) (Ability|Tasks|WaitForAttributeChangeThreshold (Variables|Default|GetAtt_Shield) "GreaterThanOrEqualTo" (Variables|Default|GetComparisonValue))
      (:OnChange
        (Variables|Default|SetShieldSuccess _bmatchescomparison)
        (if (GameplayTags|MatchesAnyTags "(TagName=\"Athena.Playlist.Phosphorus\")" (GameContext|GetGameContextGameplayTags))
          (Ability|EndAbility)
          (elif (Variables|Default|GetCountdownStarted)
            (if _bmatchescomparison
              (CallFunction|TryProgressQuest)
              (else
                (CallFunction|EndTimer)))
            (elif _bmatchescomparison
              (CallFunction|ResetCheck)
              (CallFunction|TryProgressQuest)
              (else
                (CallFunction|EndTimer)))))))))

(event Custom|TryProgressQuest
  ;[1 1 1] If shield meets threshold, send quest progress
  (Utilities|FlowControl|DoOnce false
    (:Completed
      ;;[1 0.8959 0.3865] Added this delay and additional check to prevent future occurrences of FORT-251834 In that case, the LTM flashed the HP and Shield to 100 briefly when the player jumped from the bus
      (Utilities|FlowControl|Delay 0.05)
      (if (>= (FortPawn|GetShield (Variables|GAVars|Data|GetOwningPlayerPawn)) (Variables|Default|GetComparisonValue))
        (if (not (Variables|Default|GetCountdownStarted))
          (CallFunction|StartCountdownTimer))
        (else
          (CallFunction|EndTimer))))))

(event Custom|ResetCheck
  ;[1 1 1] If shield meets threshold, send quest progress
  (if (not (Variables|Default|GetQuestComplete))
    (Utilities|FlowControl|DoOnce false
      (:Completed
        ;;[1 0.8959 0.3865] Added this delay and additional check to prevent future occurrences of FORT-251834 In that case, the LTM flashed the HP and Shield to 100 briefly when the player jumped from the bus
        (Utilities|FlowControl|Delay 0.05)
        (if (>= (FortPawn|GetShield (Variables|GAVars|Data|GetOwningPlayerPawn)) (Variables|Default|GetComparisonValue))
          (if (not (Variables|Default|GetCountdownStarted))
            (CallFunction|StartCountdownTimer))
          (else
            (CallFunction|EndTimer)))))))

(event Custom|FinishQuest
  ;[1 1 1] Start a timer when the player reaches max shields, timer is tied to transient quest granted when player gains full shield
  ; End timer if player falls below full shield
  (Quest|SendComplexCustomStatEvent (Variables|GAVars|Data|GetOwningPlayerQuestManager) 0 "(GameplayTags=((TagName=\"Athena.Quests.TimedFullHealthShields\")))" "(GameplayTags=)" 1)
  (Ability|EndAbility))

(event Custom|WinCondition
  ;[1 1 1] Start a timer when the player reaches max shields, timer is tied to transient quest granted when player gains full shield
  ; End timer if player falls below full shield
  (CallFunction|FinishQuest)
  (Variables|Default|SetQuestComplete true)
  (CallFunction|EndTimer))

(event Custom|StartCountdownTimer
  ;[1 1 1] Start a timer when the player reaches max shields, timer is tied to transient quest granted when player gains full shield
  ; End timer if player falls below full shield
  (Variables|Default|SetCountdownStarted true)
  (bind _returnvalue (Utilities|Time|SetTimerbyEvent (delegate Custom|WinCondition) (Variables|Default|GetQuestTime)))
  (CallFunction|StartCountdownQuest))

(event Custom|StartCountdownQuest
  ;[1 1 1] Start a timer when the player reaches max shields, timer is tied to transient quest granted when player gains full shield
  ; End timer if player falls below full shield
  (bind _asfort_player_controller_athena (Utilities|Casting|CastToFortPlayerControllerAthena (Fortnite|GetFortPlayerControllerfromActor (Variables|GAVars|Data|GetOwningPlayerPawn))))
  (Class|FortControllerComponentTransientQuests|GrantTransientQuest (Class|FortPlayerControllerAthena|GetTransientQuestsComponent _asfort_player_controller_athena) "/BattlepassS16/Items/QuestAbilities/Quest_FullShield1min/quest_s16_transient_fullshield1min.quest_s16_transient_fullshield1min" 0))

(event Custom|EndTimer
  ;[1 1 1] Start a timer when the player reaches max shields, timer is tied to transient quest granted when player gains full shield
  ; End timer if player falls below full shield
  (if (Variables|Default|GetCountdownStarted)
    (Utilities|Time|ClearandInvalidateTimerbyHandle (stale Custom|StartCountdownTimer _returnvalue))
    (Variables|Default|SetCountdownStarted false)
    (Variables|Default|SetShieldSuccess false)
    (CallFunction|ResetCheck)
    (CallFunction|ContinueMonitoring)))

(event Ability|EventOnEndAbility
  (params (bWasCancelled :type Boolean)))

(fn GetIsPaused
  (return false))

(fn GetIsStarted
  (return false))

(fn GetIsUrgently
  (return false))

(fn GetTimerName
  (return ""))

(fn GetTotalTime
  (return 0))

(fn GetWidgetClass
  (return 0))

(fn IsFullscreen
  (return false))
