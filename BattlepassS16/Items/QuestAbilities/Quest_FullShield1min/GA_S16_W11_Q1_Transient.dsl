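; Transient gameplay ability for a Season 16 Week 11 quest (full shield for one minute) that fails the quest when shield drops.
; On activation it waits a second, then watches the avatar pawn's CurrentShield with an attribute threshold task.
; A value below 100 leads, after another one-second delay with no recheck, to FailQuest, a debug print and EndAbility.
; The one-minute timer and success path are not in this graph.
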(event Ability|EventFailedToActivatePassiveAbility
  (params (ActorInfo :type Gameplay%20Ability%20Actor%20Info%20Structure))
  (|CallParentFunction))

(event Custom|EndQuest
  ;[1 1 1] Fail quest - server
  (CallFunction|FailQuest)
  (Development|PrintText "INVTEXT(\"Quest Failed server\")" false)
  (Ability|EndAbility))

(event Ability|EventActivateAbility
  ;[1 1 1] Get owning player pawn
  (Utilities|FlowControl|Delay 1.0)
  (bind _asfort_player_pawn (Utilities|Casting|CastToFortPlayerPawn (Ability|GetAvatarActorfromActorInfo)))
  (CallFunction|TrackDamage))

(event Custom|TrackDamage
  ;[1 1 1] Watch for shield changing to a value less than 100, fail quest if this occurs
  (bind (_asynctaskproxy _bmatchescomparison) (Ability|Tasks|WaitForAttributeChangeThreshold "(AttributeName=\"CurrentShield\",Attribute=/Script/FortniteGame.FortHealthSet:CurrentShield,AttributeOwner=None)" "LessThan" 1e+02 false (stale Ability|EventActivateAbility _asfort_player_pawn))
    (:OnChange
      (if _bmatchescomparison
        (Utilities|FlowControl|Delay 1.0)
        (CallFunction|EndQuest)))))

(event Ability|EventOnEndAbility
  (params (bWasCancelled :type Boolean)))
